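# Material function for the Athena landscape's arid sand layer, tiling textures by world XY position.
# A mask texture input drives BaseColor: blue scales the sand color, alpha blends in pebble color.
# Sand textures supply Normal, switchable roughness, Spec and Height, with Height packed into Opacity of the Mat output.
# Simple_Material uses only BaseColor and scalar specular and roughness; the collection-tinted sand color is left unconnected.
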MaterialFunction /Figment_S07/LM_7_00/Athena/Environments/Landscape/MaterialFunctions/Arid/MF_Athena_Arid_Sand
graph {nodes [
  Constant3Vector() {constant=<rgba 0 0 1 0>}
  n1 = Abs(None=ScalarParameter() {defaultValue=1500, group=Arid_Rock, parameterName=Arid_Sand_TexSize})
  n2 = ComponentMask(None=Divide(A=WorldPosition() {worldPositionShaderOffset=WPT_ExcludeAllShaderOffsets}.XYZ, B=Multiply(A=@n1) {constB=-1})) {g=true, r=true}
  n3 = TextureSampleParameter2D(UVs=@n2) {
   parameterName=Arid_Sand_01
   samplerSource=SSM_Wrap_WorldGroupSettings
   texture=/Figment_S07/LM_7_00/Environments/Sets/Arid/Textures/T_desert_sand_base_C
  }
  FunctionOutput(None=LinearInterpolate(
    A=ScalarParameter() {group=Arid_Sand, parameterName=AridSand_SpecMin}
    B=ScalarParameter() {defaultValue=0.75, group=Arid_Sand, parameterName=AridSand_SpecMax}
    Alpha=@n3.G
   )) {outputName=Spec}
  n4 = TextureObject() {samplerType=SAMPLERTYPE_LinearColor, texture=/Figment_S07/LM_7_00/Athena/Environments/Landscape/Textures/Arid/T_Athena_Arid_Rock_01_M}
  n5 = TextureSample(UVs=@n2, Tex=FunctionInput(Preview=@n4) {bUsePreviewValueAsDefault=true, inputType=FunctionInput_Texture2D}) {samplerSource=SSM_Wrap_WorldGroupSettings}
  n6 = ComponentMask(None=@n5.B) {b=true}
  n7 = VectorParameter() {parameterName=Sand_BaseColor, defaultValue=<rgba 1 0.7049 0.3585 0>}
  n8 = FunctionOutput(None=LinearInterpolate(
    A=Multiply(A=@n6, B=@n7.RGB)
    B=VectorParameter() {parameterName=Pebble_Color, defaultValue=<rgba 0.65 0.1313 0.0578 1>}.RGB
    Alpha=ComponentMask(None=@n5.A) {a=true}
   )) {outputName=BaseColor}
  FunctionOutput(None=MakeMaterialAttributes!(
    BaseColor=@n8
    Metallic
    Specular=ScalarParameter() {defaultValue=0.4, parameterName=AridSpecular}
    Roughness=ScalarParameter() {defaultValue=0.94, parameterName=Arid_Roughness}
    Anisotropy
    EmissiveColor
    Opacity
    OpacityMask
    Normal
    Tangent
    WorldPositionOffset
    SubsurfaceColor
    ClearCoat
    ClearCoatRoughness
    AmbientOcclusion
    Refraction
    CustomizedUVs_0
    CustomizedUVs_1
    CustomizedUVs_2
    CustomizedUVs_3
    CustomizedUVs_4
    CustomizedUVs_5
    CustomizedUVs_6
    CustomizedUVs_7
    PixelDepthOffset
    ShadingModel
    Displacement
   )) {outputName=Simple_Material}
  n9 = TextureSample(UVs=@n2) {
   samplerSource=SSM_Wrap_WorldGroupSettings
   samplerType=SAMPLERTYPE_Masks
   texture=/Figment_S07/LM_7_00/Environments/Sets/Arid/Textures/T_desert_sand_base_S
  }
  n10 = LinearInterpolate(
   A=ScalarParameter() {defaultValue=-0.5, group=Arid_Sand, parameterName=Sand_Rough_Min}
   B=ScalarParameter() {defaultValue=1.2, group=Arid_Sand, parameterName=Sand_Rough_Max}
   Alpha=@n9.B
  )
  n11 = Constant() {r=1}
  n12 = FunctionInput(Preview=StaticBool() {value=true}) {bUsePreviewValueAsDefault=true, inputName=UseRoughness, inputType=FunctionInput_StaticBool}
  n13 = TextureSampleParameter2D(UVs=@n2) {
   parameterName=Arid_Sand_01_Normal
   samplerSource=SSM_Wrap_WorldGroupSettings
   samplerType=SAMPLERTYPE_Normal
   texture=/Figment_S07/LM_7_00/Environments/Sets/Arid/Textures/T_desert_sand_base_N
  }
  FunctionOutput(None=MakeMaterialAttributes(
    BaseColor=@n8
    Roughness=StaticSwitch(True=@n10, False=@n11, Value=@n12)
    Opacity=FunctionOutput(None=@n3.R) {outputName=Height}
    Normal=FunctionOutput(None=@n13.RGB) {outputName=Normal}
   )) {outputName=Mat}
  Multiply(
   A=@n3.RGB
   B=ComponentMask(None=CollectionParameter() {collection=/Game/Athena/Environments/Landscape/MPC/MPC_Landscape_S07, parameterName=Arid_Sand_Tint}) {b=true, g=true, r=true}
  )
 ]}
